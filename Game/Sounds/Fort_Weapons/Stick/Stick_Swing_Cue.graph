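# Sound cue for the stick melee weapon's swing, playing one wooden whoosh per trigger.
# An equally weighted random node picks from two whoosh waves, each either plain or through an enveloper that bends pitch up then down.
# A top-level modulator randomizes pitch and halves volume.
# It uses small melee attenuation, the melee weapon sound class and a player melee concurrency set.
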SoundCue /Game/Sounds/Fort_Weapons/Stick/Stick_Swing_Cue
root=Modulator_2
nodes [
 {
  id=Enveloper_2
  children [WavePlayer_9]
  properties {pitchCurve {editorCurveData {keys [{time=0}, <at 0.3317 1.1812>, <at 0.7604 0.5858>]}}, volumeCurve {editorCurveData {keys [<at 0 1>]}}}
 }
 {
  id=Enveloper_3
  children [WavePlayer_10]
  properties {pitchCurve {editorCurveData {keys [{time=0}, <at 0.3317 1.1812>, <at 0.7604 0.5858>]}}, volumeCurve {editorCurveData {keys [<at 0 1>]}}}
 }
 {id=Modulator_2, children [Random_2], properties {pitchMax=1.1, pitchMin=0.8, volumeMax=0.5, volumeMin=0.5}}
 {id=Random_2, children [WavePlayer_9, WavePlayer_10, Enveloper_2, Enveloper_3], properties {weights [1,1,1,1]}}
 {id=WavePlayer_10, properties {soundWaveAssetPtr=/Game/Sounds/Fort_Weapons/Whooshes/Whoosh_Wood_Mix_02}}
 {id=WavePlayer_9, properties {soundWaveAssetPtr=/Game/Sounds/Fort_Weapons/Whooshes/Whoosh_Wood_Mix_01}}
]
properties {
 attenuationSettings=&Fort_Weapon_Melee_Sml_Att
 maxDistance=2800
 soundClassObject=&Fort_WeaponMelee_PSM
 volumeMultiplier=1
 attenuationOverrides {distanceAlgorithm=NaturalSound, falloffDistance=2000, attenuationShapeExtents {x=500}}
 concurrencyOverrides {maxCount=3, resolutionRule=StopOldest}
 concurrencySet [&Player_Melee_Weapon_Conc]
}
