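# Sound cue for the stick melee weapon striking wood.
# A Random node picks one of three equally weighted heavy wood impact stereo waves.
# A Modulator then varies the pitch and applies a fixed volume reduction, on top of the cue's lowered volume multiplier.
# It routes to the impacts sound class and references small melee attenuation and player melee weapon concurrency assets.
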SoundCue /Game/Sounds/Fort_Weapons/Stick/Stick_Wood_Impact
root=Modulator_4
nodes [
 {id=Modulator_4, children [Random_21], properties {pitchMax=1.2, volumeMax=0.7, volumeMin=0.7}}
 {id=Random_21, children [WavePlayer_106, WavePlayer_107, WavePlayer_108], properties {weights [1,1,1]}}
 {id=WavePlayer_106, properties {soundWaveAssetPtr=/Game/Sounds/Fort_Weapons/WeaponImpacts/Impact_WoodHeavy_Stereo_03}}
 {id=WavePlayer_107, properties {soundWaveAssetPtr=/Game/Sounds/Fort_Weapons/WeaponImpacts/Impact_WoodHeavy_Stereo_01}}
 {id=WavePlayer_108, properties {soundWaveAssetPtr=/Game/Sounds/Fort_Weapons/WeaponImpacts/Impact_WoodHeavy_Stereo_02}}
]
properties {
 attenuationSettings=&Fort_Weapon_Melee_Sml_Att
 maxDistance=2800
 soundClassObject=&Fort_Impacts
 volumeMultiplier=0.5
 attenuationOverrides {distanceAlgorithm=NaturalSound, falloffDistance=3000, attenuationShapeExtents {x=1000}}
 concurrencyOverrides {resolutionRule=PreventNew}
 concurrencySet [&Player_Melee_Weapon_Conc]
}
